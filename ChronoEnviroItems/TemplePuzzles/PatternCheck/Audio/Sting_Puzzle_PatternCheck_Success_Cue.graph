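# Sound cue for the success sting of the PatternCheck temple puzzle.
# A single wave player plays the PatternCheck puzzle sting through a modulator that fixes pitch and lowers volume, with no random variation.
# It uses the indoor ambient sound class and has a very large max distance.
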SoundCue /ChronoEnviroItems/TemplePuzzles/PatternCheck/Audio/Sting_Puzzle_PatternCheck_Success_Cue
root=Modulator_0
nodes [
 {id=Modulator_0, children [WavePlayer_0], properties {pitchMax=1, pitchMin=1, volumeMax=0.6, volumeMin=0.6}}
 {id=WavePlayer_0, properties {soundWaveAssetPtr=&/S25_Sting_Puzzle_PatternCheck}}
]
properties {maxDistance=2097152, soundClassObject=&Fort_Ambient_Indoors_Athena, volumeMultiplier=1.002}
