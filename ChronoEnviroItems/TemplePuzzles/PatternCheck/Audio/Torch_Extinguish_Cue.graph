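# Sound cue for a torch being extinguished, filed with the temple pattern-check puzzle audio.
# It picks at random among three equally weighted torch-extinguish waves, then passes the result through a modulator for variation.
# Playback uses small melee-weapon attenuation and a melee-weapon sound class, with a boosted cue volume.
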SoundCue /ChronoEnviroItems/TemplePuzzles/PatternCheck/Audio/Torch_Extinguish_Cue
root=Modulator_2
nodes [
 {id=Modulator_2, children [Random_0]}
 {id=Random_0, children [WavePlayer_0,WavePlayer_5,WavePlayer_6], properties {weights [1,1,1]}}
 {id=WavePlayer_0, properties {soundWaveAssetPtr=/Game/Sounds/Trans_AmbientLoops/Fire/Torch_Extinguish_03}}
 {id=WavePlayer_5, properties {soundWaveAssetPtr=/Game/Sounds/Trans_AmbientLoops/Fire/Torch_Extinguish_01}}
 {id=WavePlayer_6, properties {soundWaveAssetPtr=/Game/Sounds/Trans_AmbientLoops/Fire/Torch_Extinguish_02}}
]
properties {
 attenuationSettings=&Fort_Weapon_Melee_Sml_Att
 maxDistance=2800
 soundClassObject=&Fort_WeaponMelee_PSM
 volumeMultiplier=1.6
}
