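# Sound cue for a torch ignition sound in the temple pattern-check puzzle.
# A Random node picks one of three equally weighted ignite waves, and a root Modulator at default settings varies its playback.
# It borrows small melee weapon attenuation and the melee weapon sound class, with boosted volume.
# A concurrency override caps simultaneous instances and sets a short retrigger time.
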SoundCue /ChronoEnviroItems/TemplePuzzles/PatternCheck/Audio/Torch_Ignite_Cue
root=Modulator_2
nodes [
 {id=Modulator_2, children [Random_0]}
 {id=Random_0, children [WavePlayer_1,WavePlayer_2,WavePlayer_3], properties {weights [1,1,1]}}
 {id=WavePlayer_1, properties {soundWaveAssetPtr=/Game/Sounds/Trans_AmbientLoops/Fire/Torch_Ignite_01}}
 {id=WavePlayer_2, properties {soundWaveAssetPtr=/Game/Sounds/Trans_AmbientLoops/Fire/Torch_Ignite_02}}
 {id=WavePlayer_3, properties {soundWaveAssetPtr=/Game/Sounds/Trans_AmbientLoops/Fire/Torch_Ignite_03}}
]
properties {
 attenuationSettings=&Fort_Weapon_Melee_Sml_Att
 bOverrideConcurrency=true
 maxDistance=2800
 soundClassObject=&Fort_WeaponMelee_PSM
 volumeMultiplier=1.6
 concurrencyOverrides {maxCount=3, retriggerTime=0.2}
}
