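# Sound cue for a torch being extinguished, filed with the pattern-check temple puzzle audio.
# A single wave player feeds the torch extinguish sound through a modulator, with cue-level volume and pitch lowered.
# It borrows small melee weapon attenuation and a melee weapon sound class.
# A concurrency override limits simultaneous plays per owner, stopping the oldest, with a short retrigger time.
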SoundCue /ChronoEnviroItems/TemplePuzzles/PatternCheck/Audio/Torch_Interact_Extinguish_Cue
root=Modulator_2
nodes [
 {id=Modulator_2, children [WavePlayer_1]}
 {id=WavePlayer_1, properties {soundWaveAssetPtr=/Game/Sounds/Trans_AmbientLoops/Fire/Torch_Interact_Extinguish}}
]
properties {
 attenuationSettings=&Fort_Weapon_Melee_Sml_Att
 bOverrideConcurrency=true
 maxDistance=2800
 pitchMultiplier=0.8
 soundClassObject=&Fort_WeaponMelee_PSM
 volumeMultiplier=0.5
 concurrencyOverrides {bLimitToOwner=true, maxCount=3, resolutionRule=StopOldest, retriggerTime=0.2}
}
