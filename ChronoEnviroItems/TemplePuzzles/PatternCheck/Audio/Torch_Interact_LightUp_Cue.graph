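# Sound cue for a torch lighting up, used in the TemplePuzzles PatternCheck puzzle in ChronoEnviroItems.
# A single wave player feeds the Torch_Interact_LightUp fire sound into a root modulator node.
# It borrows small melee weapon attenuation and a melee weapon sound class.
# A concurrency override limits it to a few instances per owner, stopping the oldest, with a short retrigger time.
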SoundCue /ChronoEnviroItems/TemplePuzzles/PatternCheck/Audio/Torch_Interact_LightUp_Cue
root=Modulator_2
nodes [
 {id=Modulator_2, children [WavePlayer_0]}
 {id=WavePlayer_0, properties {soundWaveAssetPtr=/Game/Sounds/Trans_AmbientLoops/Fire/Torch_Interact_LightUp}}
]
properties {
 attenuationSettings=&Fort_Weapon_Melee_Sml_Att
 bOverrideConcurrency=true
 maxDistance=2800
 soundClassObject=&Fort_WeaponMelee_PSM
 volumeMultiplier=1
 concurrencyOverrides {bLimitToOwner=true, maxCount=3, resolutionRule=StopOldest, retriggerTime=0.2}
}
