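# Looping sound cue for a magical torch fire in the temple pattern-check puzzle.
# A distance crossfade blends a close layer, one of two randomly chosen fantasy torch loops with slight pitch variation, into a distant brazier loop farther away.
# Overrides add logarithmic attenuation with a custom curve and low-pass occlusion, plus a small per-owner concurrency limit.
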SoundCue /ChronoEnviroItems/TemplePuzzles/PatternCheck/Audio/Torch_MagicalFire_Loop_Cue
root=DistanceCrossFade_0
nodes [
 {
  id=DistanceCrossFade_0
  children [Modulator_2,Modulator_1]
  properties {crossFadeInput [
    {fadeInDistanceEnd=0, fadeInDistanceStart=0, fadeOutDistanceEnd=1300, fadeOutDistanceStart=500, volume=1}
    {fadeInDistanceEnd=800, fadeInDistanceStart=500, fadeOutDistanceEnd=2500, fadeOutDistanceStart=1200, volume=2}
   ]}
 }
 {id=Modulator_1, children [WavePlayer_2]}
 {id=Modulator_2, children [Random_0], properties {pitchMax=1.1, pitchMin=1, volumeMax=2, volumeMin=2}}
 {id=Random_0, children [WavePlayer_6,WavePlayer_5], properties {weights [1,1]}}
 {
  id=WavePlayer_2
  properties {bLooping=true, soundWaveAssetPtr=/Game/Sounds/Fort_Starlight/S1/SoundWaves/SpotFX_Starlight_Brazier_Distant_Loop_01}
 }
 {id=WavePlayer_5, properties {bLooping=true, soundWaveAssetPtr=&/Torch_Fantasy_loop_01}}
 {id=WavePlayer_6, properties {bLooping=true, soundWaveAssetPtr=&/Torch_Fantasy_loop_02}}
]
properties {
 bOverrideAttenuation=true
 bOverrideConcurrency=true
 maxDistance=2500
 soundClassObject=&Fort_Gadgets
 volumeMultiplier=0.6
 attenuationOverrides {
  bEnableOcclusion=true
  distanceAlgorithm=Logarithmic
  falloffDistance=1500
  focusDistanceScale=0.9
  nonFocusVolumeAttenuation=0.4
  occlusionInterpolationTime=0.4
  occlusionLowPassFilterFrequency=900
  occlusionVolumeAttenuation=0.5
  reverbDistanceMin=0
  stereoSpread=230
  attenuationShapeExtents=<500 64.5942 90>
  customAttenuationCurve {editorCurveData {keys [{arriveTangent=-2.3812, interpMode=RCIM_Cubic, leaveTangent=-2.3812, tangentMode=RCTM_User, time=0, value=1}, <at 1 0.01 RCIM_Cubic>]}}
 }
 concurrencyOverrides {bLimitToOwner=true, bVolumeScaleCanRelease=true, maxCount=2}
}
